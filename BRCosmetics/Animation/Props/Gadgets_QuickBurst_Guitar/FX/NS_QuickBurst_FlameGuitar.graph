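# Niagara system for the QuickBurst guitar gadget prop's flame effect: a short forward burst of fire with a brief orange light.
# Independent emitters layer core meshes, flame sprites, curl-noise pyro blast, cast-off fluids and colliding GPU drops.
# Spawn groups and visibility tags give core and fluid emitters two variants each; User.ControlCurve drives one core's alpha fade.
# Curve-driven dynamic material parameters animate shaders, with core meshes also computing a smoothed world-position value for lag.
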NiagaraSystem /BRCosmetics/Animation/Props/Gadgets_QuickBurst_Guitar/FX/NS_QuickBurst_FlameGuitar
dependencies {
 usedDataInterfaces [
  &NiagaraDataInterfaceCurve
  &NiagaraDataInterfaceVectorCurve
  &NiagaraDataInterfaceArrayFloat2
  &NiagaraDataInterfaceVector2DCurve
  &NiagaraDataInterfaceColorCurve
  &NiagaraDataInterfaceSpriteRendererInfo
 ]
 usedDynamicInputs []
 usedModules [&SystemState]
}
emitters {
 Cast_Off_Fluids {
  emitterUpdateScript [
   &EmitterState()
   &SpawnRate(
    SpawnRate=&FloatFromCurve(
     FloatCurve=DataInterface() {propertyValues {Curve {keys [<at 0 1 RCIM_Constant>,<at 0.9 0 RCIM_Constant>]}}}
     CurveIndex=$Emitter.NormalizedLoopAge
     Scale Curve=150
    )
    Use Spawn Probability=false
   )
   &SpawnRate(
    SpawnGroup=1
    SpawnRate=&FloatFromCurve(
     FloatCurve=DataInterface() {propertyValues {Curve {keys [<at 0 1 RCIM_Constant>,<at 0.9 0 RCIM_Constant>]}}}
     CurveIndex=$Emitter.NormalizedLoopAge
     Scale Curve=15
    )
    Use Spawn Probability=false
   ) {module=SpawnRate001}
  ]
  particleSpawnScript [
   /SetVariables(Particles.VisibilityTag=&SpawnGroup()) {isSetParameters=true}
   &InitializeParticle(
    Color=&MaskLinearColorBySpawnGroup(Masked Linear Color=<rgba 1 0.1807 0.0409 1>, Passthrough Linear Color=<rgba 1 1 1 1>)
    Lifetime=&MaskFloatBySpawnGroup(
     Masked Float=&RandomRangeFloat(Minimum=0.05, Maximum=0.4)
     Passthrough Float=&RandomRangeFloat(Minimum=0.6, Maximum=0.8)
    )
    Sprite Rotation Angle=&MaskFloatBySpawnGroup(Masked Float=0, Passthrough Float=&RandomRangeFloat(Minimum=0, Maximum=1))
    Sprite Rotation Mode=::NewEnumerator1
    Sprite Size=&MaskVector2dBySpawnGroup(
     Masked Vector2D=&RandomRangeVector2D(Minimum=<1 3>, Maximum=<3 8>)
     Passthrough Vector2D=&Vector2DFromFloat(Value=&RandomRangeFloat(Minimum=11.6667, Maximum=18.3333))
     Spawn Group Mask=0
    )
    Sprite Size Mode=::NewEnumerator4
    Sprite UV Mode=::NewEnumerator4
    Sprite UV Scale:Vector2f=&MaskVector2dBySpawnGroup(
     Masked Vector2D=<1 1>
     Passthrough Vector2D=&SelectVector2DFromArray(
      Array Sampling Mode=::NewEnumerator0
      Direct Array Index=?
      Vector 2D Selection Array=DataInterface() {propertyValues {FloatData [<-1 -1>, <1 1>]}}
      Random Seed=0
     )
     Spawn Group Mask=0
    )
    UsePositionOffset=false
    Write Material Random=true
   )
   &AddVelocity(Velocity=&RandomRangeVector(
     Minimum=<100 -30 -30>
     Maximum=&MaskVectorBySpawnGroup(Masked Vector=<550 30 100>, Passthrough Vector=<350 30 100>)
    ))
  ]
  particleUpdateScript [
   &ParticleState()
   &Drag(
    Drag=&MaskFloatBySpawnGroup(Masked Float=1.5, Passthrough Float=0)
    Use Linear Drag=true
    Use Rotational Drag=false
   )
   &GravityForce()
   &SolveForcesAndVelocity(Clamp Velocity=false, Limit Acceleration=false)
   &ScaleSpriteSizeBySpeed(Max Scale Factor=&MaskVector2dBySpawnGroup(Masked Vector2D=<1 1>, Passthrough Vector2D=<1 8>, Spawn Group Mask=0))
   &ScaleSpriteSize(
    Scale Factor=&MaskVector2dBySpawnGroup(
     Masked Vector2D=&Vector2DFromCurve(
      Vector2Curve=DataInterface() {propertyValues {
        XCurve {keys [{tangentMode=RCTM_User, time=0, value=1}, {tangentMode=RCTM_User, time=1, value=1}]}
        YCurve {keys [{tangentMode=RCTM_User, time=0, value=1}, {tangentMode=RCTM_User, time=1, value=3}]}
       }}
      ScaleCurve=<1 1>
     )
     Passthrough Vector2D=&Vector2DFromFloat(Value=&FloatFromCurve(
       FloatCurve=DataInterface() {propertyValues {Curve {keys [
           {arriveTangent=0.0076, interpMode=RCIM_Cubic, leaveTangent=0.0076, tangentMode=RCTM_User, time=0, value=0.3}
           {interpMode=RCIM_Cubic, tangentMode=RCTM_User, time=1, value=0.6}
          ]}}}
       Scale Curve=4
      ))
     Spawn Group Mask=0
    )
    Scale Sprite Size Mode=::NewEnumerator2
   )
   &DynamicMaterialParameters(
    Index 0 Param 1=&MaskFloatBySpawnGroup(
     Masked Float=&FloatFromCurve(
      FloatCurve=DataInterface() {propertyValues {Curve {keys [
          {arriveTangent=-5.8452, interpMode=RCIM_Cubic, time=0, value=1}
          {arriveTangentWeight=0.0434, interpMode=RCIM_Cubic, leaveTangentWeight=0.217, tangentMode=RCTM_User, time=0.2, value=0}
          {arriveTangent=3.4164, interpMode=RCIM_Cubic, leaveTangent=3.4164, tangentMode=RCTM_User, time=0.8, value=0.7}
          <at 1 0 RCIM_Cubic>
         ]}}}
      Scale Curve=3
     )
     Passthrough Float=&FloatFromCurve(FloatCurve=DataInterface() {propertyValues {Curve {keys [{tangentMode=RCTM_User, time=0, value=0}, {tangentMode=RCTM_User, time=1, value=1}]}}})
    )
    Index 0 Param 2=&MaskFloatBySpawnGroup(
     Masked Float=0
     Passthrough Float=&FloatFromCurve(FloatCurve=DataInterface() {propertyValues {Curve {keys [
          {tangentMode=RCTM_User, time=0, value=1}
          <at 0.05 0>
          {arriveTangent=1.5748, interpMode=RCIM_Cubic, leaveTangent=1.5748, tangentMode=RCTM_User, time=0.6, value=0.2}
          {arriveTangent=5.5221, tangentMode=RCTM_User, time=1, value=1}
         ]}}})
    )
    Index 0 Param 3=&MaskFloatBySpawnGroup(
     Masked Float=1
     Passthrough Float=&FloatFromCurve(FloatCurve=DataInterface() {propertyValues {Curve {keys [
          {arriveTangent=0.4361, interpMode=RCIM_Cubic, tangentMode=RCTM_Break, time=0.5, value=0}
          {arriveTangent=-1.6488, interpMode=RCIM_Cubic, tangentMode=RCTM_Break, time=1, value=-0.5}
         ]}}})
    )
    Index 0 Param 4=&FloatFromCurve(
     FloatCurve=DataInterface() {propertyValues {Curve {keys [
         {interpMode=RCIM_Cubic, leaveTangent=-2.006, tangentMode=RCTM_Break, time=0, value=1}
         {arriveTangent=0.0067, interpMode=RCIM_Cubic, tangentMode=RCTM_Break, time=1, value=0}
        ]}}}
     Scale Curve=15
    )
    Module.Index0Param0Write=true
    Module.Index0Param1Write=true
    Module.Index0Param2Write=true
    Module.Index0Param3Write=true
   )
   &ScaleColor(
    Scale Alpha=&MaskFloatBySpawnGroup(
     Masked Float=&FloatFromCurve(
      FloatCurve=DataInterface() {propertyValues {Curve {keys [
          {interpMode=RCIM_Cubic, tangentMode=RCTM_User, time=0, value=0}
          {arriveTangentWeight=0.0745, interpMode=RCIM_Cubic, leaveTangentWeight=0.2981, tangentMode=RCTM_User, time=0.2841, value=1}
          {interpMode=RCIM_Cubic, tangentMode=RCTM_User, time=0.5, value=1}
          {tangentMode=RCTM_User, time=1, value=0}
         ]}}}
      Scale Curve=2
     )
     Passthrough Float=&FloatFromCurve(
      FloatCurve=DataInterface() {propertyValues {Curve {keys [
          {interpMode=RCIM_Cubic, tangentMode=RCTM_User, time=0, value=0}
          {arriveTangent=0.3292, interpMode=RCIM_Cubic, leaveTangent=0.3292, time=0.1, value=1}
          {arriveTangent=-1.1111, interpMode=RCIM_Cubic, leaveTangent=-1.1111, time=0.6136, value=0.202}
          {interpMode=RCIM_Cubic, tangentMode=RCTM_User, time=1, value=0}
         ]}}}
      Scale Curve=3
     )
    )
    Scale RGB=&MaskVectorBySpawnGroup(
     Masked Vector=&VectorFromFloat(Value=&FloatFromCurve(
       FloatCurve=DataInterface() {propertyValues {Curve {keys [
           {arriveTangent=-2.3805, interpMode=RCIM_Cubic, leaveTangent=-2.3805, tangentMode=RCTM_User, time=0, value=1}
           {interpMode=RCIM_Cubic, tangentMode=RCTM_User, time=1, value=0}
          ]}}}
       Scale Curve=1.5
      ))
     Passthrough Vector=&MakeVectorFromLinearColorRGB(Color=&ColorFromCurve(ColorCurve=DataInterface() {propertyValues {
         AlphaCurve {keys [<at 0 1>]}
         BlueCurve {keys [<at 0.2932 0>,<at 0.3901 0>,<at 0.5718 0.8139>]}
         GreenCurve {keys [<at 0.2932 4.8067>,<at 0.3901 14.4201>,<at 0.5718 3.3007>]}
         RedCurve {keys [<at 0.2932 5>,<at 0.3901 15>,<at 0.5718 10>]}
        }}))
    )
    ScaleA=true
    ScaleRGB=true
   )
  ]
  rendererClasses [&NiagaraSpriteRendererProperties]
  renderers [
   {rendererClass=&NiagaraSpriteRendererProperties}
   {rendererClass=&NiagaraSpriteRendererProperties, rendererIndex=1}
  ]
 }
 Flame_Core_Meshes {
  emitterUpdateScript [
   &EmitterState()
   &SpawnBurst_Instantaneous(Spawn Time=0.05, Use Spawn Probability=false, UseLoopCountLimit=false)
   &SpawnBurst_Instantaneous(Spawn Group=1, Use Spawn Probability=false, UseLoopCountLimit=false) {module=SpawnBurst_Instantaneous001}
  ]
  particleSpawnScript [
   /SetVariables(Particles.VisibilityTag=&SpawnGroup()) {isSetParameters=true}
   &InitializeParticle(
    Color=&MaskLinearColorBySpawnGroup(Masked Linear Color=<rgba 3 3 3 0.8>, Passthrough Linear Color=<rgba 3 2.5508 1.8811 0.8>)
    Lifetime=0.5
    Mesh Scale=&MaskVectorBySpawnGroup(Masked Vector=<3 2 2>, Passthrough Vector=<1.5 0.5 0.5>)
    Mesh Scale Mode=::NewEnumerator4
    Position Offset=&MaskVectorBySpawnGroup(Masked Vector=<0 0 0>, Passthrough Vector=<-3 0 0>)
    UsePositionOffset=true
    Write Material Random=true
   )
   &CameraOffset(Camera Offset Amount=50)
   &InitialMeshOrientation(
    Mesh Axis To Orient=&MaskVectorBySpawnGroup(Masked Vector=<1 0 0>, Passthrough Vector=<-1 0 0>)
    Mesh Orientation Mode=::NewEnumerator2
    Use Rotation Vector=false
   )
  ]
  particleUpdateScript [
   &ParticleState()
   &UpdateMeshOrientation(Rotation Rate=&MaskFloatBySpawnGroup(Masked Float=0, Passthrough Float=1))
   &InterpolateOverTime(
    Rate Of Change=8
    Target Vector Value:Vector3f=&AddVectorToPosition(Position=$Particles.Position, Vector=<0 0 0>)
    Variable Type=::NewEnumerator2
   )
   /SetVariables(Particles.Interpolated Dyn Param:Vector4f=&MakeFloat4FromVectorAndFloat(
     "Vector (RGB)"=&TransformVector(
      Vector=$Particles.InterpolateOverTime.InterpolateOverTime_V3_.Moving Average
      Destination Space=::World
      Apply Scale=true
      Transform Vector as Position=true
     )
     "Float (Alpha)"=&MaskFloatBySpawnGroup(Masked Float=0.3, Passthrough Float=0.2)
    )) {isSetParameters=true}
   &DynamicMaterialParameters(
    Index 0 Param 1=&RandomRangeFloat(Minimum=0, Maximum=1)
    Index 0 Param 2=0.125
    Index 0 Param 3=2
    Index 0 Param 4=&FloatFromCurve(FloatCurve=DataInterface() {propertyValues {Curve {keys [<at 0.75 1>,<at 1 0>]}}}, Scale Curve=250)
    Index 2 Param 2=&RandomRangeFloat(Minimum=0.25, Maximum=1.25)
    Module.Index0Param0Write=true
    Module.Index0Param1Write=true
    Module.Index0Param2Write=true
    Module.Index0Param3Write=true
    Module.Index2Param0Write=true
    Module.Index2Param1Write=true
    Module.Index2Param2Write=false
    Module.Index2Param3Write=false
    Write Parameter Index 2=true
   )
   &ScaleMeshSize(Scale Factor=&VectorFromCurve(VectorCurve=DataInterface() {propertyValues {
       XCurve {keys [
         <at 0 0 RCIM_Cubic>
         {arriveTangentWeight=0.7538, interpMode=RCIM_Cubic, leaveTangentWeight=1.6584, tangentMode=RCTM_User, time=0.25, value=1}
        ]}
       YCurve {keys [{arriveTangent=3.1731, interpMode=RCIM_Cubic, leaveTangent=3.1731, tangentMode=RCTM_User, time=0, value=0.25}, <at 0.25 1 RCIM_Cubic>]}
       ZCurve {keys [{arriveTangent=3.1731, interpMode=RCIM_Cubic, leaveTangent=3.1731, tangentMode=RCTM_User, time=0, value=0.25}, <at 0.25 1 RCIM_Cubic>]}
      }}))
   &ScaleColor(
    Scale Alpha=&MaskFloatBySpawnGroup(
     Masked Float=&FloatFromCurve(FloatCurve=DataInterface() {propertyValues {Curve {keys [<at 0.8 1 RCIM_Cubic>,<at 1 0 RCIM_Cubic>]}}}, CurveIndex=$User.ControlCurve)
     Passthrough Float=&FloatFromCurve(FloatCurve=DataInterface() {propertyValues {Curve {keys [<at 0.8 1 RCIM_Cubic>, {arriveTangent=-16.5342, interpMode=RCIM_Cubic, leaveTangent=-16.5342, tangentMode=RCTM_User, time=1, value=0}]}}})
    )
    ScaleA=true
    ScaleRGB=true
   )
  ]
  rendererClasses [&NiagaraMeshRendererProperties]
  renderers [
   {rendererClass=&NiagaraMeshRendererProperties, rendererIndex=0}
   {rendererClass=&NiagaraMeshRendererProperties, rendererIndex=1}
  ]
 }
 Flame_Drops {
  simTarget=GPUComputeSim
  emitterUpdateScript [
   &EmitterState()
   &SpawnRate(
    SpawnRate=&FloatFromCurve(
     FloatCurve=DataInterface() {propertyValues {Curve {keys [<at 0 1 RCIM_Constant>,<at 0.6 0 RCIM_Constant>]}}}
     CurveIndex=$Emitter.NormalizedLoopAge
     Scale Curve=100
    )
    Use Spawn Probability=false
   )
  ]
  particleSpawnScript [
   &InitializeParticle(
    Color=<rgba 100 100 100 1>
    Lifetime Max=0.5
    Lifetime Min=0.1
    Lifetime Mode=::NewEnumerator1
    Sprite Size Max=<3 -8>
    Sprite Size Min=<1 -3>
    Sprite Size Mode=::NewEnumerator2
    UsePositionOffset=false
    Write Material Random=true
   )
   &AddVelocity(Velocity=&RandomRangeVector(Minimum=<100 -30 -30>, Maximum=<350 30 100>))
  ]
  particleUpdateScript [
   &Drag(Drag=1.5, Use Linear Drag=true, Use Rotational Drag=true)
   &GravityForce()
   &Collision(
    Age Colliding Particles=true
    Collision Enabled=&SetBoolByFloatComparison(A=$Particles.Lifetime, B=0.1)
    GPU Collision Type=::NewEnumerator2
    KillOnCollision=false
    Module.Control Rest State Via Surface Normal=false
    Module.EnableMaxCollisionCount=false
    Module.Randomize Collision Normal Vector=false
    Restitution=0
   )
   &SolveForcesAndVelocity(Clamp Velocity=false, Limit Acceleration=false)
   &ParticleState()
   &ScaleSpriteSize(
    Non-Uniform Curve Sprite Scale=DataInterface() {propertyValues {
      XCurve {keys [{tangentMode=RCTM_User, time=0, value=1}, {tangentMode=RCTM_User, time=1, value=1}]}
      YCurve {keys [{tangentMode=RCTM_User, time=0, value=1}, {tangentMode=RCTM_User, time=1, value=3}]}
     }}
    Scale Sprite Size Mode=::NewEnumerator3
   )
   &SubUVAnimation(
    End Frame Range Override=48
    "Sprite Renderer "=DataInterface() {propertyValues {SpriteRenderer=&self:Flame_Drops_5.NiagaraSpriteRendererProperties_0}}
    Start Frame Range Override=13
    UseEndFrame=true
    UseStartFrame=true
   )
  ]
  rendererClasses [&NiagaraSpriteRendererProperties]
  renderers [{rendererClass=&NiagaraSpriteRendererProperties}]
 }
 Flame_Sprites {
  emitterUpdateScript [
   &EmitterState()
   &SpawnRate(
    SpawnRate=&FloatFromCurve(
     FloatCurve=DataInterface() {propertyValues {Curve {keys [
         {arriveTangent=-2.5918, interpMode=RCIM_Constant, time=0, value=1}
         {arriveTangent=-3.1528, interpMode=RCIM_Constant, leaveTangent=-3.1528, time=0.7, value=0}
        ]}}}
     CurveIndex=$Emitter.NormalizedLoopAge
     Scale Curve=30
    )
    Use Spawn Probability=false
   )
  ]
  particleSpawnScript [
   &InitializeParticle(
    Color=<rgba 20 1.4226 0 0.55>
    Lifetime Max=0.5
    Lifetime Min=0.3
    Lifetime Mode=::NewEnumerator1
    Position Offset=<10 0 0>
    Sprite Rotation Mode=::NewEnumerator2
    Sprite Size Mode=::NewEnumerator1
    Sprite UV Mode=::NewEnumerator3
    Uniform Sprite Size Max=60
    Uniform Sprite Size Min=20
    UsePositionOffset=true
    Write Material Random=true
   )
   &AddVelocity(
    Cone Angle=10
    Module.UseConeFalloff=false
    Module.UseVelDistribution=false
    Override Seed=false
    Velocity Mode=::NewEnumerator2
    Velocity Speed=&FloatFromCurve(
     FloatCurve=DataInterface() {propertyValues {Curve {keys [{arriveTangent=-7.5919, interpMode=RCIM_Cubic, leaveTangent=-7.5919, tangentMode=RCTM_User, time=0, value=1}, <at 0.2 0.5 RCIM_Cubic>]}}}
     CurveIndex=$Emitter.LoopedAge
     Scale Curve=650
    )
   )
   &CameraOffset(Camera Offset Amount=&RandomRangeFloat(Minimum=-4, Maximum=4))
  ]
  particleUpdateScript [
   &Drag(
    Drag=&FloatFromCurve(
     FloatCurve=DataInterface() {propertyValues {Curve {keys [{tangentMode=RCTM_User, time=0.2, value=1}, {tangentMode=RCTM_User, time=0.32, value=3}]}}}
     CurveIndex=$Emitter.Age
     Scale Curve=2.5
    )
    Use Linear Drag=true
    Use Rotational Drag=false
   )
   &GravityForce(Gravity=&Multiply_VectorByFloat(
     Vector=&RandomRangeVector(Minimum=<0 0 350>, Maximum=<0 0 500>)
     Float=&FloatFromCurve(FloatCurve=DataInterface() {propertyValues {Curve {keys [<at 0 1 RCIM_Cubic>,<at 1 0 RCIM_Cubic>]}}})
    ))
   &SolveForcesAndVelocity(Clamp Velocity=false, Limit Acceleration=false)
   &ParticleState()
   &ScaleSpriteSize(Uniform Curve Sprite Scale=DataInterface() {propertyValues {Curve {keys [
        {arriveTangent=1.6361, interpMode=RCIM_Cubic, leaveTangent=1.6361, leaveTangentWeight=0.7039, tangentMode=RCTM_User, time=0, value=0.5}
        {arriveTangentWeight=0.7039, interpMode=RCIM_Cubic, tangentMode=RCTM_User, time=1, value=1.5}
       ]}}})
   &SpriteRotationRate(Rotation Rate=&RandomRangeFloat(Minimum=0, Maximum=80))
   /Niagara/Modules/Update/SubUV/SubUVAnimation(Number Of Frames=&FloatFromCurve(
     FloatCurve=DataInterface() {propertyValues {Curve {keys [{arriveTangent=4.0889, interpMode=RCIM_Cubic, leaveTangent=4.0889, tangentMode=RCTM_User, time=0, value=56}, <at 0.9 60>]}}}
     Scale Curve=&MaskFloatBySpawnGroup(Masked Float=0.7, Passthrough Float=1.25)
    ))
   &DynamicMaterialParameters(
    Index 0 Param 1=1
    Index 0 Param 2=&RandomRangeFloat(Minimum=0, Maximum=4)
    Index 0 Param 3=&FloatFromCurve(FloatCurve=DataInterface() {propertyValues {Curve {keys [<at 0 0.5 RCIM_Cubic>, {arriveTangent=1.5, interpMode=RCIM_Cubic, leaveTangent=1.5, time=0.5235, value=0.5}, <at 1 2>]}}})
    Index 0 Param 4=2.5
    Module.Index0Param0Write=true
    Module.Index0Param1Write=true
    Module.Index0Param2Write=true
    Module.Index0Param3Write=true
   )
   &ScaleColor(
    Scale Alpha=&Multiply_Float(
     A=0.8
     B=&FloatFromCurve(
      FloatCurve=DataInterface() {propertyValues {Curve {keys [
          {arriveTangent=0.162, interpMode=RCIM_Cubic, leaveTangent=0.162, tangentMode=RCTM_User, time=0, value=0.1}
          {arriveTangent=2.2098, interpMode=RCIM_Cubic, leaveTangent=2.2098, time=0.2, value=0.9}
          {arriveTangent=-0.0116, interpMode=RCIM_Cubic, leaveTangent=-0.0116, tangentMode=RCTM_User, time=0.4073, value=1}
          {arriveTangent=-2.0743, interpMode=RCIM_Cubic, leaveTangent=-2.0743, tangentMode=RCTM_User, time=0.6746, value=0.6742}
          {interpMode=RCIM_Cubic, leaveTangent=-2.371, time=1, value=0}
         ]}}}
      Scale Curve=2
     )
    )
    Scale RGB=&VectorFromFloat(Value=&FloatFromCurve(FloatCurve=DataInterface() {propertyValues {Curve {keys [<at 0 1>,<at 1 0>]}}}, Scale Curve=25))
    ScaleA=true
    ScaleRGB=true
   )
  ]
  rendererClasses [&NiagaraSpriteRendererProperties]
  renderers [{rendererClass=&NiagaraSpriteRendererProperties}]
 }
 Light {
  emitterUpdateScript [
   &EmitterState()
   &SpawnBurst_Instantaneous(Use Spawn Probability=false, UseLoopCountLimit=false)
  ]
  particleSpawnScript [&InitializeParticle(Color Mode=::NewEnumerator0, Lifetime=0.5, UsePositionOffset=false, Write Material Random=true)]
  particleUpdateScript [
   &ParticleState()
   /SetVariables(
    Particles.LightColor=&MultiplyLinearColor(
     Linear Color=<rgba 10 2.5 0 1>
     Scale Factor=&MakeFloat4FromVectorAndFloat(
      "Vector (RGB)"=<1 1 1>
      "Float (Alpha)"=&FloatFromCurve(
       FloatCurve=DataInterface() {propertyValues {Curve {keys [
           {arriveTangent=9.0725, interpMode=RCIM_Cubic, leaveTangent=9.0725, tangentMode=RCTM_User, time=0, value=0}
           {arriveTangent=1.1111, interpMode=RCIM_Cubic, leaveTangent=1.1111, time=0.1408, value=0.9847}
           {arriveTangent=-1.146, interpMode=RCIM_Cubic, leaveTangent=-1.146, time=0.9, value=1}
           {arriveTangent=-0.0405, interpMode=RCIM_Cubic, leaveTangent=-0.0405, tangentMode=RCTM_User, time=1, value=0}
          ]}}}
       Scale Curve=6
      )
     )
    )
    Particles.LightExponent=6
    Particles.LightRadius=133
   ) {isSetParameters=true}
  ]
  rendererClasses [&NiagaraLightRendererProperties]
  renderers [{rendererClass=&NiagaraLightRendererProperties, rendererIndex=0}]
 }
 Pyro_Blast_ {
  emitterUpdateScript [
   &EmitterState()
   &SpawnRate(
    SpawnRate=&FloatFromCurve(
     FloatCurve=DataInterface() {propertyValues {Curve {keys [
         {arriveTangent=-2.5918, interpMode=RCIM_Constant, time=0, value=1}
         {arriveTangent=-3.1528, interpMode=RCIM_Constant, leaveTangent=-3.1528, time=0.8, value=0}
        ]}}}
     CurveIndex=$Emitter.NormalizedLoopAge
     Scale Curve=50
    )
    Use Spawn Probability=false
   )
  ]
  particleSpawnScript [
   &InitializeParticle(
    Color=<rgba 0.3854 0.3578 0.3296 1>
    Lifetime Max=0.6
    Lifetime Min=0.3
    Lifetime Mode=::NewEnumerator1
    Position Offset=<20 0 0>
    Sprite Rotation Mode=::NewEnumerator2
    Sprite Size=&Multiply_Vector2D_ByFloat(
     Float=&FloatFromCurve(
      FloatCurve=DataInterface() {propertyValues {Curve {keys [{arriveTangent=-3.8412, interpMode=RCIM_Cubic, leaveTangent=-3.8412, tangentMode=RCTM_User, time=0, value=2}, <at 0.5 1>]}}}
      CurveIndex=$Emitter.Age
     )
     Vector2D=&Vector2DFromFloat(Value=&RandomRangeFloat(Minimum=50, Maximum=75))
    )
    Sprite Size Mode=::NewEnumerator4
    Sprite UV Mode=::NewEnumerator3
    UsePositionOffset=true
    Write Material Random=true
   )
   &AddVelocity(
    Cone Angle=10
    Module.UseConeFalloff=false
    Module.UseVelDistribution=false
    Override Seed=false
    Velocity Mode=::NewEnumerator2
    Velocity Speed=&FloatFromCurve(
     FloatCurve=DataInterface() {propertyValues {Curve {keys [{arriveTangent=-2.9253, interpMode=RCIM_Cubic, leaveTangent=-2.9253, tangentMode=RCTM_User, time=0, value=1}, <at 0.3 0.5>]}}}
     CurveIndex=$Emitter.LoopedAge
     Scale Curve=850
    )
   )
  ]
  particleUpdateScript [
   &Drag(
    Drag=&FloatFromCurve(FloatCurve=DataInterface() {propertyValues {Curve {keys [<at 0.1 1 RCIM_Cubic>,<at 1 0 RCIM_Cubic>]}}}, Scale Curve=8)
    Use Linear Drag=true
    Use Rotational Drag=false
   )
   &CurlNoiseForce(
    Noise Frequency=250
    Noise Quality / Cost=::NewEnumerator0
    Noise Strength=&FloatFromCurve(
     FloatCurve=DataInterface() {propertyValues {Curve {keys [{tangentMode=RCTM_User, time=0.3, value=0}, {tangentMode=RCTM_User, time=0.4, value=1}]}}}
     CurveIndex=$Emitter.Age
     Scale Curve=&RandomRangeFloat(Minimum=2000, Maximum=3500)
    )
    Pan Noise=true
    Pan Noise Field=<0.6 0.6 1>
   )
   &GravityForce(Gravity=&Multiply_VectorByFloat(
     Vector=&RandomRangeVector(Minimum=<0 0 350>, Maximum=<0 0 500>)
     Float=&FloatFromCurve(FloatCurve=DataInterface() {propertyValues {Curve {keys [<at 0 1 RCIM_Cubic>,<at 1 0 RCIM_Cubic>]}}})
    ))
   &SolveForcesAndVelocity(Clamp Velocity=false, Limit Acceleration=false)
   &ParticleState()
   &ScaleSpriteSize(Scale Factor=&Vector2DFromFloat(Value=&FloatFromCurve(FloatCurve=DataInterface() {propertyValues {Curve {keys [
          {arriveTangent=0.7331, interpMode=RCIM_Cubic, leaveTangent=0.7331, tangentMode=RCTM_User, time=0, value=0.25}
          {arriveTangent=2.5631, interpMode=RCIM_Cubic, leaveTangent=2.5631, tangentMode=RCTM_User, time=0.2, value=0.5}
          <at 0.7 1.25 RCIM_Cubic>
         ]}}})))
   &SpriteRotationRate(Rotation Rate=&Multiply_Float(
     A=&RandomRangeFloat(Minimum=-3, Maximum=3)
     B=&FloatFromCurve(
      FloatCurve=DataInterface() {propertyValues {Curve {keys [
          {arriveTangent=-1.9214, interpMode=RCIM_Cubic, leaveTangent=-1.9214, tangentMode=RCTM_User, time=0.2, value=1.0113}
          {arriveTangent=-0.2496, interpMode=RCIM_Cubic, leaveTangent=-0.2496, tangentMode=RCTM_User, time=0.85, value=0.1}
         ]}}}
      CurveIndex=$Emitter.Age
      Scale Curve=300
     )
    ))
   &ScaleColor(
    Scale Alpha=&FloatFromCurve(
     FloatCurve=DataInterface() {propertyValues {Curve {keys [
         <at 0 0.05>
         {arriveTangentWeight=0.846, interpMode=RCIM_Cubic, leaveTangentWeight=0.6768, tangentMode=RCTM_User, time=0.6, value=0.95}
         {interpMode=RCIM_Cubic, tangentMode=RCTM_Break, time=0.9, value=0}
        ]}}}
     Scale Curve=&FloatFromCurve(FloatCurve=DataInterface() {propertyValues {Curve {keys [{arriveTangent=-4.6812, interpMode=RCIM_Cubic, leaveTangent=-4.6812, tangentMode=RCTM_User, time=0.5, value=1}, <at 1 0 RCIM_Cubic>]}}})
    )
    ScaleA=true
    ScaleRGB=false
   )
   &DynamicMaterialParameters(
    Index 0 Param 1=&Add_Float(
     A=&RandomRangeFloat(Minimum=0, Maximum=5)
     B=&FloatFromCurve(
      FloatCurve=DataInterface() {propertyValues {Curve {keys [
          {arriveTangent=-2.2561, interpMode=RCIM_Cubic, leaveTangent=-2.2561, tangentMode=RCTM_User, time=0, value=1}
          {arriveTangent=-1.0036, interpMode=RCIM_Cubic, leaveTangent=-1.0036, tangentMode=RCTM_User, time=0.2682, value=0.3}
          <at 0.8 0>
         ]}}}
      Scale Curve=0.5
     )
    )
    Index 0 Param 2=&FloatFromCurve(
     FloatCurve=DataInterface() {propertyValues {Curve {keys [
         {arriveTangent=0.086, interpMode=RCIM_Cubic, leaveTangent=0.086, tangentMode=RCTM_User, time=0.15, value=0.2}
         {arriveTangent=2.0528, interpMode=RCIM_Cubic, leaveTangent=2.0528, tangentMode=RCTM_User, time=1, value=1}
        ]}}}
     CurveIndex=$Emitter.Age
     Scale Curve=0.4
    )
    Index 0 Param 3=1
    Index 1 Param 1=&FloatFromCurve(
     FloatCurve=DataInterface() {propertyValues {Curve {keys [
         {arriveTangent=-1.3223, interpMode=RCIM_Cubic, leaveTangent=-1.3223, tangentMode=RCTM_User, time=0.1, value=0.5}
         {arriveTangentWeight=0.3, interpMode=RCIM_Cubic, time=1, value=0}
        ]}}}
     Scale Curve=&FloatFromCurve(FloatCurve=DataInterface() {propertyValues {Curve {keys [<at 0.1 15>,<at 0.3 1>,<at 1 1>]}}})
    )
    Index 1 Param 2=&FloatFromCurve(
     FloatCurve=DataInterface() {propertyValues {Curve {keys [{arriveTangent=-1.2376, interpMode=RCIM_Cubic, leaveTangent=-1.2376, tangentMode=RCTM_User, time=0, value=1}, <at 1 0>]}}}
     Scale Curve=10
    )
    Index 1 Param 3=&FloatFromCurve(
     FloatCurve=DataInterface() {propertyValues {Curve {keys [
         {interpMode=RCIM_Cubic, tangentMode=RCTM_User, time=0.333, value=1.5}
         {arriveTangent=0.7571, interpMode=RCIM_Cubic, leaveTangent=0.7571, tangentMode=RCTM_User, time=1, value=1.8}
        ]}}}
     CurveIndex=$Emitter.Age
     Scale Curve=1.35
    )
    Index 1 Param 4=&FloatFromCurve(
     FloatCurve=DataInterface() {propertyValues {Curve {keys [<at 0 0.5>,<at 1 0.3>]}}}
     Scale Curve=&Add_Float(A=1, B=&RandomRangeFloat(Minimum=0, Maximum=5))
    )
    Index 2 Param 1=&FloatFromCurve(FloatCurve=DataInterface() {propertyValues {Curve {keys [{arriveTangent=0.0048, interpMode=RCIM_Cubic, leaveTangent=0.0048, tangentMode=RCTM_User, time=0, value=0.8}, <at 0.9 1.2 RCIM_Cubic>]}}})
    Index 2 Param 2=&Add_Float(
     A=&RandomRangeFloat(Minimum=0, Maximum=100)
     B=&FloatFromCurve(
      FloatCurve=DataInterface() {propertyValues {Curve {keys [
          {arriveTangent=0.5501, interpMode=RCIM_Cubic, leaveTangent=0.5501, tangentMode=RCTM_User, time=0, value=0}
          {arriveTangent=1.1924, interpMode=RCIM_Cubic, leaveTangent=1.1924, tangentMode=RCTM_User, time=1, value=0.8}
         ]}}}
      Scale Curve=&RandomRangeFloat(Minimum=-0.3, Maximum=-0.56)
     )
    )
    Index 2 Param 3=&FloatFromCurve(FloatCurve=DataInterface() {propertyValues {Curve {keys [
         {arriveTangent=-0.4045, interpMode=RCIM_Cubic, tangentMode=RCTM_User, time=0.4, value=0.5}
         {arriveTangent=5.0157, interpMode=RCIM_Cubic, leaveTangent=5.0157, tangentMode=RCTM_User, time=1, value=2}
        ]}}})
    Index 2 Param 4=&FloatFromCurve(FloatCurve=DataInterface() {propertyValues {Curve {keys [{arriveTangent=-0.4045, time=0, value=0.8}, {leaveTangent=-0.318, time=1, value=1}]}}})
    Index 3 Param 1=&FloatFromCurve(
     FloatCurve=DataInterface() {propertyValues {Curve {keys [
         {arriveTangent=0.1645, interpMode=RCIM_Cubic, leaveTangent=0.1645, tangentMode=RCTM_User, time=0.1, value=0.2}
         {arriveTangent=2.6313, interpMode=RCIM_Cubic, leaveTangent=2.6313, tangentMode=RCTM_User, time=0.6, value=1.25}
         <at 1 2 RCIM_Cubic>
        ]}}}
     Scale Curve=4
    )
    Index 3 Param 2=&FloatFromCurve(
     FloatCurve=DataInterface() {propertyValues {Curve {keys [<at 0 1>,<at 1 0>]}}}
     Scale Curve=&RandomRangeFloat(Minimum=0.4, Maximum=0.5)
    )
    Index 3 Param 3=&FloatFromCurve(
     FloatCurve=DataInterface() {propertyValues {Curve {keys [
         {arriveTangent=0.4884, interpMode=RCIM_Cubic, leaveTangent=0.4884, tangentMode=RCTM_User, time=0, value=0.7}
         {arriveTangent=0.3564, interpMode=RCIM_Cubic, leaveTangent=0.3564, tangentMode=RCTM_User, time=1, value=0.9}
        ]}}}
     Scale Curve=0.0009
    )
    Index 3 Param 4=1
    Module.Index0Param0Write=true
    Module.Index0Param1Write=true
    Module.Index0Param2Write=true
    Module.Index0Param3Write=false
    Module.Index1Param0Write=true
    Module.Index1Param1Write=true
    Module.Index1Param2Write=true
    Module.Index1Param3Write=true
    Module.Index2Param0Write=true
    Module.Index2Param1Write=true
    Module.Index2Param2Write=true
    Module.Index2Param3Write=true
    Module.Index3Param0Write=true
    Module.Index3Param1Write=true
    Module.Index3Param2Write=true
    Module.Index3Param3Write=true
    Write Parameter Index 1=true
    Write Parameter Index 2=true
    Write Parameter Index 3=true
   )
  ]
  rendererClasses [&NiagaraSpriteRendererProperties]
  renderers [{rendererClass=&NiagaraSpriteRendererProperties}]
 }
}
summary {
 systemName=NS_QuickBurst_FlameGuitar
 emitters [
  {emitterName=Flame_Core_Meshes, rendererClasses [&NiagaraMeshRendererProperties]}
  {emitterName=Cast_Off_Fluids, rendererClasses [&NiagaraSpriteRendererProperties]}
  {emitterName=Flame_Drops, simTarget=GPUComputeSim, rendererClasses [&NiagaraSpriteRendererProperties]}
  {emitterName=Pyro_Blast_, rendererClasses [&NiagaraSpriteRendererProperties]}
  {emitterName=Flame_Sprites, rendererClasses [&NiagaraSpriteRendererProperties]}
  {emitterName=Light, rendererClasses [&NiagaraLightRendererProperties]}
 ]
 userVariables [User.ControlCurve {type=NiagaraFloat, defaultValue=0, description=""}]
}
