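# Niagara idle effect for the QuickBurst flame guitar gadget prop, with a single Gas_Leak sprite emitter.
# It continuously spawns short-lived, randomly sized sprites that drift along X, grow and fade out over their lifetime.
# User.GasControl blends a dynamic material parameter between a constant and a subtly dipping curve.
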NiagaraSystem /BRCosmetics/Animation/Props/Gadgets_QuickBurst_Guitar/FX/NS_QuickBurst_FlameGuitar_Idle
dependencies {usedDataInterfaces [&NiagaraDataInterfaceCurve], usedDynamicInputs [], usedModules [&SystemState]}
emitters {Gas_Leak {
  emitterUpdateScript [
   &EmitterState()
   &SpawnRate(SpawnRate=14, Use Spawn Probability=false)
  ]
  particleSpawnScript [
   &InitializeParticle(
    Lifetime Max=0.5
    Lifetime Min=0.4
    Lifetime Mode=::NewEnumerator1
    Sprite Rotation Mode=::NewEnumerator2
    Sprite Size Mode=::NewEnumerator1
    Sprite UV Mode=::NewEnumerator3
    Uniform Sprite Size Max=30
    Uniform Sprite Size Min=15
    UsePositionOffset=false
    Write Material Random=true
   )
   &AddVelocity(Velocity=<50 0 0>)
  ]
  particleUpdateScript [
   &ParticleState()
   &ScaleSpriteSize(Uniform Curve Sprite Scale=DataInterface() {propertyValues {Curve {keys [{arriveTangent=2.1708, interpMode=RCIM_Cubic, leaveTangent=2.1708, tangentMode=RCTM_User, time=0, value=0.25}, <at 1 1 RCIM_Cubic>]}}})
   &SolveForcesAndVelocity(Clamp Velocity=false, Limit Acceleration=false)
   &DynamicMaterialParameters(
    Index 0 Param 1=&Lerp_Float(
     A=1
     B=&FloatFromCurve(FloatCurve=DataInterface() {propertyValues {Curve {keys [
          {interpMode=RCIM_Cubic, tangentMode=RCTM_User, time=0, value=1}
          {interpMode=RCIM_Cubic, tangentMode=RCTM_User, time=0.4, value=0.96}
          {interpMode=RCIM_Cubic, tangentMode=RCTM_User, time=0.75, value=1}
         ]}}})
     Alpha=$User.GasControl
    )
    Module.Index0Param0Write=true
    Module.Index0Param1Write=false
    Module.Index0Param2Write=false
    Module.Index0Param3Write=false
   )
   &ScaleColor(
    Scale Alpha=&FloatFromCurve(FloatCurve=DataInterface() {propertyValues {Curve {keys [<at -0.0046 1.0113 RCIM_Cubic>,<at 1.0011 0.002 RCIM_Cubic>]}}})
    ScaleA=true
    ScaleRGB=true
   )
  ]
  rendererClasses [&NiagaraSpriteRendererProperties]
  renderers [{rendererClass=&NiagaraSpriteRendererProperties}]
 }}
summary {
 systemName=NS_QuickBurst_FlameGuitar_Idle
 emitters [{emitterName=Gas_Leak, rendererClasses [&NiagaraSpriteRendererProperties]}]
 userVariables [User.GasControl {type=NiagaraFloat, defaultValue=1, description=""}]
}
